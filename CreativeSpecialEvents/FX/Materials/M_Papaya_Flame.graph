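# Additive unlit flame material for Niagara and particle sprites, using the Papaya gadget color function to tint the fire.
# A fireball SubUV flipbook is desaturated by a parameter, multiplied by the gadget color and particle color, then boosted and passed through a shared FX emissive adjustment function.
# The same desaturated flipbook times particle alpha drives a depth fade with adjustable distance to soften intersections with geometry.
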Material /CreativeSpecialEvents/FX/Materials/M_Papaya_Flame
graph {
 properties {
  bUsedWithNiagaraSprites=true
  bUsedWithParticleSprites=true
  blendMode=BLEND_Additive
  pixelDepthOffsetMode=PDOM_Legacy
  shadingModel=MSM_Unlit
  usedShadingModels=Unlit
 }
 nodes [
  n1 = Desaturation(
   None=ParticleSubUV!(UVs, Tex, Apply View MipBias) {texture=/Game/Effects/Fort_Effects/Textures/T_Fire_Ball_Sub_UV3}.RGB
   Fraction=ScalarParameter() {parameterName=Desat}
  )
  n2 = ParticleColor()
  DepthFade(Opacity=Multiply(A=@n1, B=@n2.A), FadeDistance=ScalarParameter() {defaultValue=30, parameterName=DepthFade_Distance}) {fadeDistanceDefault=30}
  n3 = Multiply(
   A=MaterialFunctionCall!(
    Media Player TexCoord
    Media Color Code Coordinate
    Media Texture Resolution
    Base Color Parameter
    Base Color Mix
    Media Color Code Mix
    Intermission Color Parameter
    Intermission Color Mix
    Desaturation
   ) {materialFunction=/Game/Environments/Papaya/Materials/MaterialFunctions/MF_Papaya_GadgetColor}.Result
   B=@n1
  )
  MaterialFunctionCall(Emissive=Multiply(A=Multiply(A=@n2.RGB, B=@n3)) {constB=4}) {materialFunction=&MF_FX_EmissiveAdjustments}
 ]
}
