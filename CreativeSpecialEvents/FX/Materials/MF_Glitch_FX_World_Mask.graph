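# Material function that generates an animated world-space mask for a glitch effect, plus a camera distance mask.
# Scaled world position feeds layered, time-scrolled sine waves and a rotator, shaping a sphere mask whose radius pulses via DebugTimeSine calls.
# The mask's vector length gates a fast sine flicker, scaled by MinGlitchiness, as WorldSpaceMask.
# DistanceFadeMask rises from zero at the camera to one at WPOOffsetFadeDist.
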MaterialFunction /CreativeSpecialEvents/FX/Materials/MF_Glitch_FX_World_Mask
graph {nodes [
  n1 = Multiply(A=Time(), B=ScalarParameter() {defaultValue=1.7, parameterName=FlickerAddSpeed})
  n2 = Add(A=Multiply(A=Sine(None=@n1) {period=0.1}, B=ScalarParameter() {defaultValue=0.2, parameterName=FlickerMult}))
  n3 = Constant() {r=1}
  n4 = Divide(
   A=WorldPosition().XYZ
   B=Multiply(
    A=FunctionInput(Preview=@n3) {bUsePreviewValueAsDefault=true, inputName=MaskScaleMult, inputType=FunctionInput_Scalar}
    B=ScalarParameter() {defaultValue=6000, parameterName=WorldMaskScale}
   )
  )
  n5 = Sine(None=ComponentMask(None=@n4) {r=true}) {period=1.7}
  n6 = Multiply(A=Add(A=Multiply(A=@n5) {constB=0.25}, B=Time()), B=ScalarParameter() {defaultValue=3, parameterName=WaveSpeed})
  n7 = Add(A=@n6, B=@n4)
  n8 = Multiply(A=@n6) {constB=0.8}
  n9 = ConstantBiasScale(None=Sine(None=Add(A=@n8, B=@n4)) {period=0.51})
  n10 = Add(A=Multiply(A=@n6) {constB=-0.25}, B=@n4)
  n11 = Add(
   A=LinearInterpolate(A=Sine(None=@n7) {period=0.8}, B=Rotator(Coordinate=@n9) {centerX=0, centerY=0, speed=1})
   B=Multiply(A=Sine(None=@n10) {period=0.42}) {constB=0.6}
  )
  n12 = ConstantBiasScale(None=Sine(None=@n10))
  n13 = ScalarParameter() {defaultValue=0.25, parameterName=PhaseSpeed}
  n14 = Add(A=Round(None=@n12), B=ScalarParameter() {defaultValue=12, parameterName=FlickerSpeed})
  n15 = LinearInterpolate(
   A=OneMinus(None=@n12)
   B=@n12
   Alpha=LinearInterpolate(
    A=MaterialFunctionCall(Speed=@n13) {materialFunction=&DebugTimeSine}.Result
    B=MaterialFunctionCall(Speed=@n14) {materialFunction=&DebugTimeSine}.Result
   ) {constAlpha=0.2}
  ) {constA=-1}
  n16 = ScalarParameter() {defaultValue=3, parameterName=WPOMaskRadius}
  n17 = MaterialFunctionCall(Vector 3=SphereMask(A=ConstantBiasScale(None=@n11), B=Constant(), Radius=Multiply(A=@n15, B=@n16) {constB=2}) {attenuationRadius=2, hardnessPercent=0}) {materialFunction=/Engine/Functions/Engine_MaterialFunctions02/Utility/VectorLength}
  FunctionOutput(None=Multiply(A=ScalarParameter() {defaultValue=0.5, parameterName=MinGlitchiness}, B=LinearInterpolate(B=@n2, Alpha=@n17.V3 Length))) {outputName=WorldSpaceMask}
  n18 = Distance(A=WorldPosition().XYZ, B=CameraPositionWS())
  n19 = ScalarParameter() {defaultValue=80000, parameterName=WPOOffsetFadeDist}
  FunctionOutput(None=Saturate(None=Divide(A=@n18, B=@n19) {constB=5000})) {outputName=DistanceFadeMask, sortPriority=1}
 ]}
