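# Additive unlit mesh-particle material for Papaya event lightning bolts, with dynamic parameters driving travel, width, wobble and random phase.
# Vertices push along normals with near-camera shrink, wobble through step-quantized scrolling noise, and jitter along a meteor direction inside a glitch world mask.
# A travelling sphere mask, eroded by noise as particle alpha drops, shapes the bolt.
# Emissive blends event collection and gadget colors, adds glitch flicker, and adjusts for time of day, quality and mobile.
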Material /CreativeSpecialEvents/FX/Materials/M_Papaya_LightningBolt_Mesh
graph {
 properties {
  bUsedWithMeshParticles=true
  bUsedWithNiagaraMeshParticles=true
  blendMode=BLEND_Additive
  pixelDepthOffsetMode=PDOM_Legacy
  shadingModel=MSM_Unlit
  usedShadingModels=Unlit
 }
 nodes [
  n1 = ScalarParameter() {defaultValue=50, parameterName=Push}
  n2 = DynamicParameter() {defaultValue=<rgba 0.5 1 1 1>, paramNames [Travel,WidthMult,WobbleMult,Random]}
  n3 = VertexColor()
  n4 = Multiply(A=Normalize(VectorInput=VertexNormalWS()), B=Multiply(A=Multiply(A=@n1, B=@n2.WidthMult), B=LinearInterpolate(Alpha=@n3.R) {constA=0.5}))
  n5 = MaterialFunctionCall(Distance=ScalarParameter() {defaultValue=25000, parameterName=NearCameraShrink}) {materialFunction=&Near_Camera_Fade}
  n6 = MaterialFunctionCall!(MaskScaleMult=ScalarParameter() {defaultValue=3, parameterName=GlitchMaskScaleMult}) {materialFunction=&/MF_Glitch_FX_World_Mask}
  n7 = Multiply(A=@n6.WorldSpaceMask) {constB=0.4}
  n8 = TextureCoordinate() {uTiling=0}
  n9 = ScalarParameter() {defaultValue=1, parameterName=WobbleTile}
  n10 = Time()
  n11 = ScalarParameter() {defaultValue=0.1, parameterName=WobbleSpeed}
  n12 = TextureSample(UVs=Panner(Coordinate=Multiply(A=@n8, B=@n9), Time=Multiply(A=@n10, B=@n11)) {speedX=0.2, speedY=2}) {samplerType=SAMPLERTYPE_LinearColor, texture=&HighContrastNoise}
  n13 = ScalarParameter() {defaultValue=4, parameterName=WobbleStepSize}
  n14 = Divide(A=Round(None=Multiply(A=@n12.RGB, B=@n13) {constB=10}), B=@n13) {constB=10}
  n15 = Multiply(
   A=Multiply(A=Add(A=@n2.WobbleMult, B=@n7), B=ScalarParameter() {defaultValue=500, parameterName=WobbleStrength})
   B=Multiply(A=Subtract(A=@n14) {constB=0.5}) {constB=2}
  )
  n16 = Add(A=Time(), B=DynamicParameter() {defaultValue=<rgba 0.5 1 1 1>, paramNames [Travel,WidthMult,WobbleMult,Random]}.Random)
  n17 = VectorParameter() {parameterName=MeteorDirection, defaultValue=<rgba 0.7755 -0.4565 -0.4361 0>}
  n18 = ScalarParameter() {defaultValue=120, parameterName=GlitchWPODist}
  n19 = Round(None=Multiply(A=Sine(None=@n16) {period=0.11}, B=Multiply(A=@n17.RGB, B=@n18)))
  Add(
   A=Add(A=Multiply(A=@n4, B=@n5.Result), B=Multiply(A=@n3.G, B=@n15))
   B=Multiply(A=LinearInterpolate(Alpha=@n6.DistanceFadeMask) {constB=3}, B=Multiply(A=@n6.WorldSpaceMask, B=@n19))
  )
  n20 = TextureCoordinate() {uTiling=0, vTiling=12}
  n21 = ParticleColor()
  n22 = Multiply(A=TextureSample(UVs=Panner(Coordinate=@n20, Time=@n21.A) {speedY=-4}) {samplerType=SAMPLERTYPE_LinearColor, texture=&HighContrastNoise}.R) {constB=-1}
  n23 = OneMinus(None=@n21.A)
  n24 = ComponentMask(None=TextureCoordinate()) {g=true}
  n25 = LinearInterpolate(Alpha=@n2.Travel) {constA=-1, constB=2}
  Clamp(None=Add(A=Multiply(A=@n22, B=@n23), B=SphereMask(A=@n24, B=@n25, Hardness=@n2.Travel) {attenuationRadius=1, hardnessPercent=50}))
  n26 = Constant() {r=0.64}
  n27 = MaterialFunctionCall(Color_Morning=@n26, Color_Day=Constant() {r=1}, Color_Evening=@n26, Color_Night=Constant() {r=0.2}) {materialFunction=&Color_Time_of_Day}
  n28 = CollectionParameter() {
   collection=&FortniteMaterialParameters
   parameterName=PapayaBaseColorParameter
  }
  n29 = CollectionParameter() {
   collection=&FortniteMaterialParameters
   parameterName=PapayaBaseColorParameterOffset
  }
  n30 = LinearInterpolate(
   A=Add(A=@n28, B=@n29)
   B=MaterialFunctionCall!(
    Media Player TexCoord
    Media Color Code Coordinate
    Media Texture Resolution
    Base Color Parameter
    Base Color Mix
    Media Color Code Mix
    Intermission Color Parameter
    Intermission Color Mix
    Desaturation
   ) {materialFunction=/Game/Environments/Papaya/Materials/MaterialFunctions/MF_Papaya_GadgetColor}.Result
   Alpha=@n3.R
  )
  n31 = VectorParameter() {parameterName=Glitch2, defaultValue=<rgba 1 0 0.2301 0>}
  n32 = Multiply(A=ConstantBiasScale(None=Sine(None=@n16) {period=0.178}), B=Multiply(A=Multiply(A=@n6.WorldSpaceMask, B=@n6.WorldSpaceMask)) {constB=2})
  n33 = Clamp(None=Add(A=Multiply(A=@n21.RGB, B=@n30), B=Multiply(A=@n31.RGB, B=@n32)))
  n34 = ScalarParameter() {defaultValue=10, parameterName=HDRMult}
  n35 = Multiply(A=@n33, B=QualitySwitch(Default=ScalarParameter() {defaultValue=2, parameterName=HDRMult_Default}, High=@n34, Epic=@n34))
  n36 = MaterialFunctionCall(LDR Input=@n33, HDR Tint=@n35, Desaturate Luminance=Constant() {r=0.7}) {materialFunction=&DeriveHDRfromLDR}
  MaterialFunctionCall(Emissive=ShadingPathSwitch(Default=Multiply(A=@n27.Result, B=@n36.RemappedRange), Mobile=Clamp(None=@n35) {maxDefault=20})) {materialFunction=&MF_FX_EmissiveAdjustments}
 ]
}
