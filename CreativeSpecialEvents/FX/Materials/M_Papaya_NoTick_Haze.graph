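# Additive unlit haze sprite material for Papaya special events that reacts to event audio amplitude.
# Audio amplitude blends time-of-day color toward gadget color and adds dust motes to panning clouds, with motes dropped on cheaper quality and shading paths.
# A soft flare mask shapes opacity, softened by depth, near-camera and distance fades.
# World position offset builds camera-distance-scaled sprites, offset by vertex color, with optional curl-noise drift.
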Material /CreativeSpecialEvents/FX/Materials/M_Papaya_NoTick_Haze
graph {
 properties {
  blendMode=BLEND_Additive
  opacityMaskClipValue=0.05
  pixelDepthOffsetMode=PDOM_Legacy
  shadingModel=MSM_Unlit
  translucencyLightingMode=TLM_VolumetricPerVertexNonDirectional
  usedShadingModels=Unlit
 }
 nodes [
  n1 = CollectionParameter() {collection=&FortniteMaterialParameters, parameterName=EventAudio_Amplitude}
  n2 = Constant() {r=1}
  n3 = ScalarParameter() {parameterName=Debug}
  n4 = Saturate(None=Multiply(A=LinearInterpolate(A=@n1, B=@n2, Alpha=@n3), B=ScalarParameter() {defaultValue=7, parameterName=Amplitude Multiplier}))
  n5 = LinearInterpolate(
   A=MaterialFunctionCall(
    Color_Morning=VectorParameter() {parameterName=Color Morning, defaultValue=<rgba 1 1 1 1>}.RGB
    Color_Day=VectorParameter() {parameterName=Color Day, defaultValue=<rgba 1 1 1 1>}.RGB
    Color_Evening=VectorParameter() {parameterName=Color Evening, defaultValue=<rgba 1 1 1 1>}.RGB
    Color_Night=VectorParameter() {parameterName=Color Night, defaultValue=<rgba 1 1 1 1>}.RGB
   ) {materialFunction=&Color_Time_of_Day}.Result
   B=MaterialFunctionCall!(
    Media Player TexCoord
    Media Color Code Coordinate
    Media Texture Resolution
    Base Color Parameter
    Base Color Mix
    Media Color Code Mix
    Intermission Color Parameter
    Intermission Color Mix
    Desaturation
   ) {materialFunction=/Game/Environments/Papaya/Materials/MaterialFunctions/MF_Papaya_GadgetColor}.Result
   Alpha=@n4
  )
  n6 = ShadingPathSwitch(
   Default=Multiply(A=@n5, B=ScalarParameter() {defaultValue=1, parameterName=Emissive}) {constB=0.1}
   Forward=Multiply(A=@n5, B=ScalarParameter() {defaultValue=0.35, parameterName=Forward Render Multiplier})
  )
  n7 = TextureSample(UVs=Panner(Coordinate=TextureCoordinate() {uTiling=1.15, vTiling=-1}) {speedY=-0.04}) {samplerSource=SSM_Wrap_WorldGroupSettings, texture=/Game/Packages/Fortnite_UniqueMaterials_1/PostProcess/T_TilingClouds_02}
  n8 = TextureCoordinate() {uTiling=1.5, vTiling=1.5}
  n9 = TextureSample(UVs=Panner(Coordinate=Add(A=@n8) {constB=0.85}) {speedY=0.0167}) {samplerSource=SSM_Wrap_WorldGroupSettings, texture=/Game/Global/Textures/Default/PointSplatter/T_Tiling_Dust}
  n10 = TextureCoordinate()
  n11 = TextureSample(UVs=Panner(Coordinate=Add(A=@n10) {constB=0.2}) {speedX=0.0083, speedY=0.0275}) {samplerSource=SSM_Wrap_WorldGroupSettings, texture=&T_Tiling_Dust}
  n12 = Saturate(None=Multiply(A=Multiply(A=@n9.G, B=@n11.G), B=ScalarParameter() {defaultValue=40, parameterName=Mote Scale}) {constB=40})
  n13 = Add(A=@n7.G, B=LinearInterpolate(B=Add(A=@n12) {constB=0.05}, Alpha=@n4))
  n14 = QualitySwitch(Default=@n13, Low=@n7.G, High=@n13, Medium=@n7.G)
  n15 = TextureSample() {texture=/Game/Global/Textures/Default/Masks_and_Gradients/T_New_Soft_Flair}
  n16 = Add(A=Multiply(A=ShadingPathSwitch(Default=@n14, Forward=@n7.G, Mobile=@n7.G), B=@n15.G), B=@n15.G)
  n17 = ScalarParameter() {defaultValue=800, parameterName=DepthFadeDistance}
  n18 = Multiply(A=@n17) {constB=4}
  n19 = Clamp(None=@n16)
  n20 = Multiply(
   A=DepthFade(Opacity=@n16, FadeDistance=LinearInterpolate(A=@n17, B=@n18, Alpha=@n19))
   B=MaterialFunctionCall(Distance=ScalarParameter() {defaultValue=500, parameterName=Near Camera Fade Distance}) {materialFunction=&Near_Camera_Fade}.Result
  )
  n21 = ScalarParameter() {defaultValue=0.3714, parameterName=Opacity}
  Multiply(
   A=MaterialFunctionCall(LDR Input=Saturate(None=@n6), HDR Tint=@n6, HDR Intensity Multiplier=ScalarParameter() {defaultValue=4, parameterName=HDR Multiplier}) {materialFunction=&DeriveHDRfromLDR}.RemappedRange
   B=Multiply(A=Multiply(A=@n20, B=@n21), B=DistanceCullFade())
  )
  n22 = WorldPosition()
  n23 = Multiply(A=Time() {period=16}, B=ScalarParameter() {defaultValue=0.15, parameterName=Slow})
  n24 = VectorNoise(World Position=Multiply(A=VertexNormalWS(), B=Add(A=@n22.XYZ, B=@n23))) {bTiling=true, noiseFunction=VNF_CurlALU, tileSize=1}
  n25 = Multiply(A=VertexColor().A, B=ScalarParameter() {defaultValue=3, parameterName=ScaleOffset})
  n26 = StaticSwitchParameter(True=Transform(None=Multiply(A=@n24, B=@n25)) {transformSourceType=TRANSFORMSOURCE_Local}, False=Constant()) {defaultValue=true, parameterName=AnimateSprite}
  n27 = MaterialFunctionCall(Distance=ScalarParameter() {defaultValue=484.8872, parameterName=SoftenDistance}) {materialFunction=&Near_Camera_Fade}
  n28 = Multiply(
   A=StaticSwitchParameter(True=VertexColor().A, False=Constant() {r=1}) {defaultValue=true, parameterName=Use Vertex Color Alpha Mask for Sprite Scale}
   B=LinearInterpolate(
    A=ScalarParameter() {defaultValue=450, parameterName=SpriteScaleNear}
    B=ScalarParameter() {defaultValue=620.4763, parameterName=SpriteScaleFar}
    Alpha=Clamp(None=@n27.Result)
   )
  )
  n29 = MaterialFunctionCall!(Center Location=ObjectPositionWS(), 0-1 UVs, XY scale=AppendVector(A=@n28, B=@n28), OverridePivot, "Normals (Optional)") {materialFunction=&Sprite}
  n30 = VertexColor()
  n31 = VectorParameter() {parameterName=OffsetX, defaultValue=<rgba 360.3565 0 0 0>}
  n32 = VectorParameter() {parameterName=OffsetY, defaultValue=<rgba -312.1444 0 0 0>}
  n33 = Add(
   A=Add(A=Multiply(A=@n30.R, B=@n31.RGB), B=Multiply(A=@n30.G, B=@n32.RGB))
   B=Multiply(A=@n30.B, B=VectorParameter() {parameterName=OffsetZ, defaultValue=<rgba 0 0 67.9512 0>}.RGB)
  )
  Reroute(None=Add(A=Add(A=@n26, B=@n29.World position offset), B=Transform(None=@n33) {transformSourceType=TRANSFORMSOURCE_Local}))
 ]
}
