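# Unlit additive spark material for sprite particles in Papaya special-event effects.
# The spark texture's red channel, with optionally swizzled UVs, is multiplied by particle color and alpha, with a brightness boost at low quality.
# That is tinted by MF_Papaya_GadgetColor and passed through shared emissive adjustments.
# A near-camera fade feeds depth fade on the default shading path, while mobile skips the depth fade.
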Material /CreativeSpecialEvents/FX/Materials/M_Papaya_Spark
graph {
 properties {
  allowTranslucentCustomDepthWrites=true
  bEnableResponsiveAA=true
  bUsedWithNiagaraSprites=true
  bUsedWithParticleSprites=true
  blendMode=BLEND_Additive
  opacityMaskClipValue=0.01
  pixelDepthOffsetMode=PDOM_Legacy
  shadingModel=MSM_Unlit
  usedShadingModels=Unlit
 }
 nodes [
  n1 = TextureCoordinate()
  n2 = TextureSampleParameter2D(UVs=LinearInterpolate(
    A=TextureCoordinate()
    B=MaterialFunctionCall(XY=@n1) {materialFunction=&Swizzle}.YX
    Alpha=ScalarParameter() {parameterName=SwizzleUV}
   )) {parameterName=Texture, texture=/Game/Weapons/FORT_Rifles/Effects/Textures/spark}
  n3 = ParticleColor()
  n4 = Constant() {r=1.6}
  n5 = QualitySwitch(Default=Multiply(A=@n2.R, B=@n3.RGB), Low=Multiply(A=Multiply(A=@n4, B=@n2.R), B=@n3.RGB))
  MaterialFunctionCall(Emissive=Multiply(
    A=MaterialFunctionCall!(
     Media Player TexCoord
     Media Color Code Coordinate
     Media Texture Resolution
     Base Color Parameter
     Base Color Mix
     Media Color Code Mix
     Intermission Color Parameter
     Intermission Color Mix
     Desaturation
    ) {materialFunction=/Game/Environments/Papaya/Materials/MaterialFunctions/MF_Papaya_GadgetColor}.Result
    B=Multiply(A=@n5, B=@n3.A)
   )) {materialFunction=&MF_FX_EmissiveAdjustments}
  n6 = MaterialFunctionCall(Distance=ScalarParameter() {defaultValue=200, parameterName=Near Camera Distance}) {materialFunction=&Near_Camera_Fade}
  ShadingPathSwitch(
   Default=DepthFade(Opacity=@n6.Result, FadeDistance=ScalarParameter() {defaultValue=50, parameterName=Depth Fade Distance})
   Mobile=@n6.Result
  )
 ]
}
