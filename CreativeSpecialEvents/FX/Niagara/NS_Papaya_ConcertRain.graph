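# Niagara system for a concert rain effect: colored flame streaks and GPU sparks falling from a wide overhead cylinder, some sparks pushed by turbulence vector fields.
# FlameSprites collide, age quickly and generate collision events; generated flames, impact sparks and a lit flare have no spawn modules, so likely respond to those events.
# Every emitter samples the user color curve using the Papaya jukebox parameter collection's base color value, syncing tint system-wide.
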NiagaraSystem /CreativeSpecialEvents/FX/Niagara/NS_Papaya_ConcertRain
dependencies {
 usedDataInterfaces [
  &NiagaraDataInterfaceCurve
  &NiagaraDataInterfaceVector2DCurve
  /Script/Niagara.NiagaraDataInterfaceVectorField
 ]
 usedDynamicInputs []
 usedModules [&SystemState]
}
emitters {
 FlameSprites {
  emitterUpdateScript [
   &EmitterState(
    Life Cycle Mode=::NewEnumerator1
    Loop Delay=&UniformRangedFloat(Minimum=0.5, Maximum=2)
    Loop Duration=&UniformRangedFloat(Minimum=0.5, Maximum=1)
    Recalculate Duration Each Loop=true
    UseLoopDelay=true
   )
   &SpawnRate(SpawnRate=40, Use Spawn Probability=false)
  ]
  particleSpawnScript [
   /Niagara/Modules/Spawn/Initialization/InitializeParticle(
    Color=&MakeLinearColorFromVectorAndFloat("Vector (RGB)"=<1 1 1>, "Float (Alpha)"=&UniformRangedFloat(Minimum=0.1, Maximum=0.35))
    Lifetime=&UniformRangedFloat(Minimum=1.6, Maximum=2)
    Module.Write Color=true
    Module.Write Lifetime=true
    Module.Write Mass=true
    Module.Write Material Random=true
    Module.Write Position=false
    Module.Write Scale=false
    Module.Write SpriteRotation=false
    Module.Write SpriteSize=true
    Position=&SimulationPosition()
    Sprite Size=&UniformRangedVector2D(Minimum=<8 80>, Maximum=<22 150>)
   )
   &CylinderLocation(
    Cylinder Height=2500
    Cylinder Radius=&UniformRangedFloat(Minimum=500, Maximum=6800)
    Lathe Profile=DataInterface() {propertyValues {Curve {keys []}}}
    MaskSpawnGroup=false
    Module.Enable Lathe Profile=false
    Module.Override Randomness=false
    Module.Surface Only=true
    Offset=<0 0 800>
    Surface Only Band Thickness=1
    Use Endcaps In Surface Only Mode=false
   )
   &AddVelocity(Velocity=&UniformRangedVector(Minimum=<150 150 -800>, Maximum=<250 250 -1600>))
  ]
  particleUpdateScript [
   &ParticleState()
   &AccelerationForce(Acceleration=<30 30 -400>)
   &Collision(
    Advanced Aging Rate=10
    Age Colliding Particles=true
    Friction=1
    Friction During a Bounce=1
    Module.Control Rest State Via Surface Normal=false
    Module.EnableMaxCollisionCount=false
    Module.Randomize Collision Normal Vector=true
    Restitution=0
   )
   &SolveForcesAndVelocity(Clamp Velocity=false, Limit Acceleration=false)
   &ScaleSpriteSize(Scale Factor=&Vector2DFromCurve(
     Vector2Curve=DataInterface() {propertyValues {
       XCurve {keys [
         <at 0 0.8 RCIM_Cubic>
         {arriveTangent=2.2852, interpMode=RCIM_Cubic, leaveTangent=2.2852, tangentMode=RCTM_User, time=0.4, value=1.25}
         <at 1 2.3 RCIM_Cubic>
        ]}
       YCurve {keys [
         <at 0 0.3 RCIM_Cubic>
         {arriveTangent=4.5192, interpMode=RCIM_Cubic, leaveTangent=4.5192, tangentMode=RCTM_User, time=0.4, value=1.65}
         <at 1 3.6 RCIM_Cubic>
        ]}
      }}
     ScaleCurve=<1 1>
    ))
   &ScaleColor(
    Scale RGB=&MakeVectorFromLinearColorRGB(Color=&ColorFromCurve(ColorCurve=$User.BaseColorCurve, CurveIndex=&Frac_Float(Float=$NPC.Papaya_Jukebox_NPC.Base Color Parameter)))
    ScaleA=true
    ScaleRGB=true
   )
   &ScaleColor(
    Scale RGB=&VectorFromFloat(Value=&FloatFromCurve(FloatCurve=DataInterface() {propertyValues {Curve {keys [<at 0 1>,<at 0.7 10>]}}}))
    ScaleA=true
    ScaleRGB=true
   ) {module=ScaleColor001}
   /Niagara/Modules/Events/GenerateCollisionEvent(Module.Use Time Between Events=false, Module.Use Velocity Threshold=true, Velocity Threshold=0)
  ]
  rendererClasses [&NiagaraSpriteRendererProperties]
  renderers [{rendererClass=&NiagaraSpriteRendererProperties}]
 }
 FlameSprites_Generated {
  emitterUpdateScript [&EmitterState(Life Cycle Mode=::NewEnumerator1, Loop Duration=1, UseLoopDelay=false)]
  particleSpawnScript [
   &InitializeParticle(
    Lifetime=&UniformAOrBFloat(A=0.3, B=0.7)
    Module.Write Color=true
    Module.Write Lifetime=true
    Module.Write Mass=true
    Module.Write Material Random=true
    Module.Write Position=false
    Module.Write Scale=false
    Module.Write SpriteRotation=true
    Module.Write SpriteSize=true
    Position=&SimulationPosition()
    Sprite Rotation=&UniformRangedFloat(Minimum=0, Maximum=360)
    Sprite Size=&Vector2DFromFloat(Value=&UniformRangedFloat(Minimum=80, Maximum=150))
   )
   &CameraOffset(Camera Offset Amount=80)
   /SetVariables(Particles.UVScale:Vector2f=&MakeVector2D(X=&UniformAOrBFloat(A=-1), Y=&UniformAOrBFloat(A=-1))) {isSetParameters=true}
  ]
  particleUpdateScript [
   &ParticleState()
   &Drag(Drag=4, Use Linear Drag=true, Use Rotational Drag=false)
   &SolveForcesAndVelocity(Clamp Velocity=false, Limit Acceleration=false)
   &ScaleColor(
    Scale RGB=&MakeVectorFromLinearColorRGB(Color=&ColorFromCurve(ColorCurve=$User.BaseColorCurve, CurveIndex=&Frac_Float(Float=$NPC.Papaya_Jukebox_NPC.Base Color Parameter)))
    ScaleA=true
    ScaleRGB=true
   )
   &ScaleColor(
    Scale Alpha=&Lerp_Float(A=1, B=0, Alpha=$Particles.NormalizedAge)
    Scale RGB=&VectorFromFloat(Value=&FloatFromCurve(FloatCurve=DataInterface() {propertyValues {Curve {keys [<at 0 20>,<at 0.7 5>]}}}))
    ScaleA=true
    ScaleRGB=true
   ) {module=ScaleColor001}
   &ScaleSpriteSize(Scale Factor=&Vector2DFromFloat(Value=&FloatFromCurve(FloatCurve=DataInterface() {propertyValues {Curve {keys [
          {arriveTangent=2.6522, interpMode=RCIM_Cubic, leaveTangent=2.6522, tangentMode=RCTM_User, time=0, value=0}
          {arriveTangent=1.7692, interpMode=RCIM_Cubic, leaveTangent=1.7692, tangentMode=RCTM_User, time=0.2, value=0.5}
          <at 1 1 RCIM_Cubic>
         ]}}})))
  ]
  rendererClasses [&NiagaraSpriteRendererProperties]
  renderers [{rendererClass=&NiagaraSpriteRendererProperties}]
 }
 High_Sparks {
  simTarget=GPUComputeSim
  emitterUpdateScript [
   &EmitterState(
    Life Cycle Mode=::NewEnumerator1
    Loop Delay=&UniformRangedFloat(Minimum=0.3, Maximum=0.5)
    Loop Duration=&UniformRangedFloat(Minimum=1, Maximum=2)
    Recalculate Duration Each Loop=true
    UseLoopDelay=true
   )
   &SpawnRate(SpawnRate=5600, Use Spawn Probability=false)
  ]
  particleSpawnScript [
   &InitializeParticle(
    Lifetime=&UniformRangedFloat(Minimum=0.5, Maximum=1)
    Module.Write Color=true
    Module.Write Lifetime=true
    Module.Write Mass=true
    Module.Write Material Random=true
    Module.Write Position=false
    Module.Write Scale=false
    Module.Write SpriteRotation=false
    Module.Write SpriteSize=true
    Position=&SimulationPosition()
    Sprite Size=&UniformRangedVector2D(Minimum=<3 2>, Maximum=<4.5 8>)
   )
   &CylinderLocation(
    Cylinder Height=2500
    Cylinder Radius=&UniformRangedFloat(Minimum=250, Maximum=4000)
    Lathe Profile=DataInterface() {propertyValues {Curve {keys []}}}
    MaskSpawnGroup=false
    Module.Enable Lathe Profile=false
    Module.Override Randomness=false
    Module.Surface Only=true
    Offset=<0 0 800>
    Surface Only Band Thickness=1
    Use Endcaps In Surface Only Mode=false
   )
   &AddVelocity(Velocity=&UniformRangedVector(Minimum=<150 150 -800>, Maximum=<250 250 -1600>))
  ]
  particleUpdateScript [
   &ParticleState()
   /Niagara/VectorFields/SampleVectorField(
    Sampled Vector Scale=3
    VectorField=DataInterface() {propertyValues {Field=/Game/Effects/VectorFields/Noisy_Turbulence_103, bTileX=true, bTileY=true, bTileZ=true}}
   )
   /Niagara/VectorFields/ApplyVectorField(Vector Field Intensity as Force=500, Vector Field Velocity Tightness=0.1)
   &AccelerationForce(Acceleration=<500 500 -200>)
   &Drag(Drag=23, Use Linear Drag=true, Use Rotational Drag=false)
   &SolveForcesAndVelocity(Clamp Velocity=false, Limit Acceleration=false)
   &ScaleSpriteSize(Scale Factor=&Vector2DFromCurve(
     Vector2Curve=DataInterface() {propertyValues {XCurve {keys [<at 0 0.66>,<at 0.6 0.6>]}, YCurve {keys [<at 0 1>,<at 0.6 0.5>]}}}
     ScaleCurve=<1 1>
    ))
   &ScaleSpriteSizeBySpeed(Max Scale Factor=<1 3>)
   &ScaleColor(
    Scale RGB=&MakeVectorFromLinearColorRGB(Color=&ColorFromCurve(ColorCurve=$User.BaseColorCurve, CurveIndex=&Frac_Float(Float=$NPC.Papaya_Jukebox_NPC.Base Color Parameter)))
    ScaleA=true
    ScaleRGB=true
   )
   &ScaleColor(
    Scale Alpha=&Lerp_Float(A=1, B=0, Alpha=$Particles.NormalizedAge)
    Scale RGB=&VectorFromFloat(Value=&FloatFromCurve(FloatCurve=DataInterface() {propertyValues {Curve {keys [<at 0 20>,<at 0.7 5>]}}}))
    ScaleA=true
    ScaleRGB=true
   ) {module=ScaleColor001}
  ]
  rendererClasses [&NiagaraSpriteRendererProperties]
  renderers [{rendererClass=&NiagaraSpriteRendererProperties}]
 }
 IMpact_Flare {
  emitterUpdateScript [&EmitterState(Life Cycle Mode=::NewEnumerator1, Loop Delay=0.1, Loop Duration=1, UseLoopDelay=true)]
  particleSpawnScript [&InitializeParticle(
    Lifetime=0.2
    Module.Write Color=true
    Module.Write Lifetime=true
    Module.Write Mass=true
    Module.Write Material Random=true
    Module.Write Position=false
    Module.Write Scale=false
    Module.Write SpriteRotation=true
    Module.Write SpriteSize=true
    Position=&SimulationPosition()
    Sprite Size=<0 0>
   )]
  particleUpdateScript [
   &ParticleState()
   &SolveForcesAndVelocity(Clamp Velocity=false, Limit Acceleration=false)
   &ScaleSpriteSize(Scale Factor=&Vector2DFromFloat(Value=&Lerp_Float(A=1, B=0, Alpha=$Particles.NormalizedAge)))
   &ScaleColor(
    Scale RGB=&MakeVectorFromLinearColorRGB(Color=&ColorFromCurve(ColorCurve=$User.BaseColorCurve, CurveIndex=&Frac_Float(Float=$NPC.Papaya_Jukebox_NPC.Base Color Parameter)))
    ScaleA=true
    ScaleRGB=true
   )
   &ScaleColor(
    Scale Alpha=&FloatFromCurve(FloatCurve=DataInterface() {propertyValues {Curve {keys [{arriveTangent=-9.8188, interpMode=RCIM_Cubic, time=0, value=0.7}, <at 1 0 RCIM_Cubic>]}}})
    Scale RGB=&VectorFromFloat(Value=&FloatFromCurve(FloatCurve=DataInterface() {propertyValues {Curve {keys [<at 0 4>,<at 0.05 1>]}}}))
    ScaleA=true
    ScaleRGB=true
   ) {module=ScaleColor001}
   &FN_Light_v2(Light Exponent=3, Light Radius=250, LightEnabled=true)
  ]
  rendererClasses [&NiagaraSpriteRendererProperties, &NiagaraLightRendererProperties]
  renderers [
   {rendererClass=&NiagaraSpriteRendererProperties}
   {rendererClass=&NiagaraLightRendererProperties, rendererIndex=1}
  ]
 }
 Sparks {
  emitterUpdateScript [&EmitterState(Life Cycle Mode=::NewEnumerator1, Loop Delay=0.04, Loop Duration=1, UseLoopDelay=true)]
  particleSpawnScript [&InitializeParticle(
    Color=<rgba 10 10 10 1>
    Lifetime=&UniformRangedFloat(Minimum=0.3, Maximum=1.5)
    Module.Write Color=true
    Module.Write Lifetime=true
    Module.Write Mass=true
    Module.Write Material Random=true
    Module.Write Position=false
    Module.Write Scale=false
    Module.Write SpriteRotation=false
    Module.Write SpriteSize=true
    Position=&SimulationPosition()
    Sprite Size=&UniformRangedVector2D(Minimum=<3 5>, Maximum=<5 8>)
   )]
  particleUpdateScript [
   &ParticleState()
   &SampleVectorField(
    FieldRotate:Quat4f=&LerpQuaternion(
     Quaternion A:Quat4f=<1 0 0 0>
     Quaternion B:Quat4f=<0 0 0 0>
     Lerp Factor:NiagaraFloat=&UniformRangedFloat(Minimum=0, Maximum=1)
     Enable Spherical Linear Interpolation:NiagaraBool=true
    )
    Sampled Vector Scale=3
    VectorField=DataInterface() {propertyValues {Field=&Noisy_Turbulence_103, bTileX=true, bTileY=true, bTileZ=true}}
   )
   &ApplyVectorField(
    Vector Field Intensity as Velocity:NiagaraFloat=&Multiply_Float(A=80, B=&UniformRangedFloat(Minimum=0, Maximum=0.3))
    Vector Field Velocity Tightness=0.1
   )
   &AccelerationForce(Acceleration=<0 0 -600>)
   &SolveForcesAndVelocity(Clamp Velocity=false, Limit Acceleration=false)
   &ScaleSpriteSize(Scale Factor=&Vector2DFromCurve(
     Vector2Curve=DataInterface() {propertyValues {XCurve {keys [<at 0 0.66>,<at 0.6 0.6>]}, YCurve {keys [<at 0 1>,<at 0.6 0.5>]}}}
     ScaleCurve=<1 1>
    ))
   &ScaleColor(
    Scale RGB=&MakeVectorFromLinearColorRGB(Color=&ColorFromCurve(ColorCurve=$User.BaseColorCurve, CurveIndex=&Frac_Float(Float=$NPC.Papaya_Jukebox_NPC.Base Color Parameter)))
    ScaleA=true
    ScaleRGB=true
   )
   &ScaleColor(
    Scale Alpha=&Lerp_Float(A=1, B=0, Alpha=$Particles.NormalizedAge)
    Scale RGB=&VectorFromFloat(Value=&FloatFromCurve(FloatCurve=DataInterface() {propertyValues {Curve {keys [<at 0 20>,<at 0.7 5>]}}}))
    ScaleA=true
    ScaleRGB=true
   ) {module=ScaleColor001}
  ]
  rendererClasses [&NiagaraSpriteRendererProperties]
  renderers [{rendererClass=&NiagaraSpriteRendererProperties}]
 }
 Sparks_Rain {
  simTarget=GPUComputeSim
  emitterUpdateScript [
   &EmitterState(
    Life Cycle Mode=::NewEnumerator1
    Loop Delay=&UniformRangedFloat(Minimum=0.3, Maximum=0.5)
    Loop Duration=&UniformRangedFloat(Minimum=1, Maximum=2)
    Recalculate Duration Each Loop=true
    UseLoopDelay=true
   )
   &SpawnRate(SpawnRate=1500, Use Spawn Probability=false)
  ]
  particleSpawnScript [
   &InitializeParticle(
    Lifetime=&UniformRangedFloat(Minimum=0.5, Maximum=1)
    Module.Write Color=true
    Module.Write Lifetime=true
    Module.Write Mass=true
    Module.Write Material Random=true
    Module.Write Position=false
    Module.Write Scale=false
    Module.Write SpriteRotation=false
    Module.Write SpriteSize=true
    Position=&SimulationPosition()
    Sprite Size=&UniformRangedVector2D(Minimum=<3 5>, Maximum=<4 8>)
   )
   &CylinderLocation(
    Cylinder Height=2500
    Cylinder Radius=&UniformRangedFloat(Minimum=250, Maximum=4000)
    Lathe Profile=DataInterface() {propertyValues {Curve {keys []}}}
    MaskSpawnGroup=false
    Module.Enable Lathe Profile=false
    Module.Override Randomness=false
    Module.Surface Only=true
    Offset=<0 0 800>
    Surface Only Band Thickness=1
    Use Endcaps In Surface Only Mode=false
   )
   &AddVelocity(Velocity=&UniformRangedVector(Minimum=<150 150 -800>, Maximum=<250 250 -1600>))
  ]
  particleUpdateScript [
   &ParticleState()
   &AccelerationForce(Acceleration=<0 0 -400>)
   &SolveForcesAndVelocity(Clamp Velocity=false, Limit Acceleration=false)
   &ScaleSpriteSize(Scale Factor=&Vector2DFromCurve(
     Vector2Curve=DataInterface() {propertyValues {XCurve {keys [<at 0 0.66>,<at 0.6 0.6>]}, YCurve {keys [<at 0 1>,<at 0.6 0.5>]}}}
     ScaleCurve=<1 1>
    ))
   &ScaleSpriteSizeBySpeed(Max Scale Factor=<1 20>)
   &ScaleColor(
    Scale RGB=&MakeVectorFromLinearColorRGB(Color=&ColorFromCurve(ColorCurve=$User.BaseColorCurve, CurveIndex=&Frac_Float(Float=$NPC.Papaya_Jukebox_NPC.Base Color Parameter)))
    ScaleA=true
    ScaleRGB=true
   )
   &ScaleColor(
    Scale Alpha=&Lerp_Float(A=1, B=0, Alpha=$Particles.NormalizedAge)
    Scale RGB=&VectorFromFloat(Value=&FloatFromCurve(FloatCurve=DataInterface() {propertyValues {Curve {keys [<at 0 20>,<at 0.7 5>]}}}))
    ScaleA=true
    ScaleRGB=true
   ) {module=ScaleColor001}
  ]
  rendererClasses [&NiagaraSpriteRendererProperties]
  renderers [{rendererClass=&NiagaraSpriteRendererProperties}]
 }
 Sparks_ShortLived {
  emitterUpdateScript [&EmitterState(Life Cycle Mode=::NewEnumerator1, Loop Delay=0.04, Loop Duration=1, UseLoopDelay=true)]
  particleSpawnScript [&InitializeParticle(
    Color=<rgba 10 10 10 1>
    Lifetime=&UniformRangedFloat(Minimum=0.05, Maximum=0.1)
    Module.Write Color=true
    Module.Write Lifetime=true
    Module.Write Mass=true
    Module.Write Material Random=true
    Module.Write Position=false
    Module.Write Scale=false
    Module.Write SpriteRotation=false
    Module.Write SpriteSize=true
    Position=&SimulationPosition()
    Sprite Size=&UniformRangedVector2D(Minimum=<2 5>, Maximum=<4 8>)
   )]
  particleUpdateScript [
   &ParticleState()
   &SampleVectorField(
    FieldRotate:Quat4f=&LerpQuaternion(
     Quaternion A:Quat4f=<1 0 0 0>
     Quaternion B:Quat4f=<0 0 0 0>
     Lerp Factor:NiagaraFloat=&UniformRangedFloat(Minimum=0, Maximum=1)
     Enable Spherical Linear Interpolation:NiagaraBool=true
    )
    Sampled Vector Scale=3
    VectorField=DataInterface() {propertyValues {Field=&Noisy_Turbulence_103, bTileX=true, bTileY=true, bTileZ=true}}
   )
   &ApplyVectorField(
    Vector Field Intensity as Velocity:NiagaraFloat=&Multiply_Float(A=80, B=&UniformRangedFloat(Minimum=0, Maximum=0.3))
    Vector Field Velocity Tightness=0.1
   )
   &AccelerationForce(Acceleration=<0 0 -600>)
   &SolveForcesAndVelocity(Clamp Velocity=false, Limit Acceleration=false)
   &ScaleSpriteSize(Scale Factor=&Vector2DFromCurve(
     Vector2Curve=DataInterface() {propertyValues {XCurve {keys [<at 0 0.66>,<at 0.6 0.6>]}, YCurve {keys [<at 0 1>,<at 0.6 0.5>]}}}
     ScaleCurve=<1 1>
    ))
   &ScaleSpriteSizeBySpeed(Max Scale Factor=<1 20>)
   &ScaleColor(
    Scale RGB=&MakeVectorFromLinearColorRGB(Color=&ColorFromCurve(ColorCurve=$User.BaseColorCurve, CurveIndex=&Frac_Float(Float=$NPC.Papaya_Jukebox_NPC.Base Color Parameter)))
    ScaleA=true
    ScaleRGB=true
   )
   &ScaleColor(
    Scale Alpha=&Lerp_Float(A=1, B=0, Alpha=$Particles.NormalizedAge)
    Scale RGB=&VectorFromFloat(Value=&FloatFromCurve(FloatCurve=DataInterface() {propertyValues {Curve {keys [<at 0 20>,<at 0.7 5>]}}}))
    ScaleA=true
    ScaleRGB=true
   ) {module=ScaleColor001}
  ]
  rendererClasses [&NiagaraSpriteRendererProperties]
  renderers [{rendererClass=&NiagaraSpriteRendererProperties}]
 }
}
summary {
 systemName=NS_Papaya_ConcertRain
 emitters [
  {emitterName=Sparks_Rain, simTarget=GPUComputeSim, rendererClasses [&NiagaraSpriteRendererProperties]}
  {emitterName=High_Sparks, simTarget=GPUComputeSim, rendererClasses [&NiagaraSpriteRendererProperties]}
  {emitterName=FlameSprites, rendererClasses [&NiagaraSpriteRendererProperties]}
  {emitterName=Sparks, rendererClasses [&NiagaraSpriteRendererProperties]}
  {emitterName=Sparks_ShortLived, rendererClasses [&NiagaraSpriteRendererProperties]}
  {emitterName=FlameSprites_Generated, rendererClasses [&NiagaraSpriteRendererProperties]}
  {emitterName=IMpact_Flare, rendererClasses [&NiagaraSpriteRendererProperties, &NiagaraLightRendererProperties]}
 ]
 userVariables [User.BaseColorCurve {
   type=NiagaraDataInterfaceColorCurve
   description=""
   defaultValue {
    struct=NiagaraExt_VariableValue_DataInterface
    value {
     dataInterface=&self:NiagaraDataInterfaceColorCurve_0
     dataInterfaceClass=&NiagaraDataInterfaceColorCurve
    }
   }
  }]
}
